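# Physics asset for the Blackwidow cosmetic backpack, defining simulation bodies and joints for its eight arms.
# A capsule on backpack_root anchors four mirrored arm pairs (top, mid top, mid bottom, bottom), each a chain of three capsule bodies.
# Constraints allow bending mainly along swing1, with swing2 and twist nearly locked for hinge-like motion.
# The outer joints of the mid-bottom arms get wider swing limits than the rest.
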PhysicsAsset /BRCosmetics/Accessories/FORT_Backpacks/F_MED_Blackwidow_Backpack/Meshes/F_MED_Blackwidow_Backpack_01_Physics
bodies {
 arm_bot_02_l [{shapeType=Capsule, length=6.3178, radius=1.9765, center=<-2.5763 -0.0001 0>, rotation=<rot 90 0 0>}]
 arm_bot_02_r [{shapeType=Capsule, length=6.0678, radius=1.9765, center=<2.1665 0.0002 0>, rotation=<rot 90 0 0>}]
 arm_bot_03_l [{shapeType=Capsule, length=5.1631, radius=2.045, center=<-0.1865 -0.0002 0>, rotation=<rot 90 0 0>}]
 arm_bot_03_r [{shapeType=Capsule, length=6.1631, radius=2.045, center=<0.1727 0 0>, rotation=<rot 90 0 0>}]
 arm_bot_04_l [{shapeType=Capsule, length=5.0891, radius=1.8776, center=<1.029 0.0001 0>, rotation=<rot 90 0 0>}]
 arm_bot_04_r [{shapeType=Capsule, length=5.0891, radius=1.8776, center=<-0.5174 0.0002 0>, rotation=<rot 90 0 0>}]
 arm_mid_bot_02_l [{shapeType=Capsule, length=3.911, radius=2.7402, center=<-1.5525 0 0>, rotation=<rot 90 0 0>}]
 arm_mid_bot_02_r [{shapeType=Capsule, length=4.161, radius=2.4902, center=<1.3733 0 0>, rotation=<rot 90 0 0>}]
 arm_mid_bot_03_l [{shapeType=Capsule, length=3.0906, radius=2.6709, center=<-1.2445 0 0>, rotation=<rot 90 0 0>}]
 arm_mid_bot_03_r [{shapeType=Capsule, length=3.0906, radius=2.6709, center=<0.6728 0 0>, rotation=<rot 90 0 0>}]
 arm_mid_bot_04_l [{shapeType=Capsule, length=2.6913, radius=2.5856, center=<-1.7772 0 0>, rotation=<rot 90 0 0>}]
 arm_mid_bot_04_r [{shapeType=Capsule, length=2.6913, radius=2.5856, center=<1.1155 0.0002 0>, rotation=<rot 90 0 0>}]
 arm_mid_top_02_l [{shapeType=Capsule, length=3.5532, radius=1.8189, center=<-0.9406 0 0.0001>, rotation=<rot 90 0 0>}]
 arm_mid_top_02_r [{shapeType=Capsule, length=3.5532, radius=1.8189, center=<1.1613 0 0.0001>, rotation=<rot 90 0 0>}]
 arm_mid_top_03_l [{shapeType=Capsule, length=4.855, radius=2.3364, center=<-0.8997 0 0>, rotation=<rot 90 0 0>}]
 arm_mid_top_03_r [{shapeType=Capsule, length=4.855, radius=2.3364, center=<0.8997 0 0>, rotation=<rot 90 0 0>}]
 arm_mid_top_04_l [{shapeType=Capsule, length=4.5802, radius=1.7182, center=<-1.2274 0 0>, rotation=<rot 90 0 0>}]
 arm_mid_top_04_r [{shapeType=Capsule, length=4.5802, radius=1.7182, center=<1.2274 0 0>, rotation=<rot 90 0 0>}]
 arm_top_02_l [{shapeType=Capsule, length=6.9928, radius=2.0287, center=<-2.2044 0 0.0001>, rotation=<rot 90 0 0>}]
 arm_top_02_r [{shapeType=Capsule, length=6.9928, radius=2.0287, center=<1.3113 0 0>, rotation=<rot 90 0 0>}]
 arm_top_03_l [{shapeType=Capsule, length=6.0929, radius=1.8368, center=<-1.1761 0 -0.0001>, rotation=<rot 90 0 0>}]
 arm_top_03_r [{shapeType=Capsule, length=6.0929, radius=1.8368, center=<1.0836 0 0>, rotation=<rot 90 0 0>}]
 arm_top_04_l [{shapeType=Capsule, length=5.045, radius=1.9973, center=<-0.4671 0 0>, rotation=<rot 90 0 0>}]
 arm_top_04_r [{shapeType=Capsule, length=5.045, radius=1.9973, center=<1.0729 0 0.0001>, rotation=<rot 90 0 0>}]
 backpack_root [{shapeType=Capsule, length=5.0873, radius=3.4501}]
}
constraints [
 {bone1Name=arm_top_02_r, bone2Name=backpack_root, swing1LimitDegrees=20, swing2LimitDegrees=1, twistLimitDegrees=1}
 {bone1Name=arm_top_03_r, bone2Name=arm_top_02_r, swing1LimitDegrees=20, swing2LimitDegrees=1, twistLimitDegrees=1}
 {bone1Name=arm_top_04_r, bone2Name=arm_top_03_r, swing1LimitDegrees=20, swing2LimitDegrees=1, twistLimitDegrees=1}
 {bone1Name=arm_mid_top_02_r, bone2Name=backpack_root, swing1LimitDegrees=20, swing2LimitDegrees=1, twistLimitDegrees=1}
 {bone1Name=arm_mid_top_03_r, bone2Name=arm_mid_top_02_r, swing1LimitDegrees=20, swing2LimitDegrees=1, twistLimitDegrees=1}
 {bone1Name=arm_mid_top_04_r, bone2Name=arm_mid_top_03_r, swing1LimitDegrees=20, swing2LimitDegrees=1, twistLimitDegrees=1}
 {bone1Name=arm_mid_bot_02_r, bone2Name=backpack_root, swing1LimitDegrees=20, swing2LimitDegrees=1, twistLimitDegrees=1}
 {bone1Name=arm_mid_bot_03_r, bone2Name=arm_mid_bot_02_r, swing1LimitDegrees=35, swing2LimitDegrees=1, twistLimitDegrees=1}
 {bone1Name=arm_mid_bot_04_r, bone2Name=arm_mid_bot_03_r, swing1LimitDegrees=40, swing2LimitDegrees=1, twistLimitDegrees=1}
 {bone1Name=arm_bot_02_r, bone2Name=backpack_root, swing1LimitDegrees=20, swing2LimitDegrees=1, twistLimitDegrees=1}
 {bone1Name=arm_bot_03_r, bone2Name=arm_bot_02_r, swing1LimitDegrees=20, swing2LimitDegrees=1, twistLimitDegrees=1}
 {bone1Name=arm_bot_04_r, bone2Name=arm_bot_03_r, swing1LimitDegrees=20, swing2LimitDegrees=1, twistLimitDegrees=1}
 {bone1Name=arm_bot_02_l, bone2Name=backpack_root, swing1LimitDegrees=20, swing2LimitDegrees=1, twistLimitDegrees=1}
 {bone1Name=arm_bot_03_l, bone2Name=arm_bot_02_l, swing1LimitDegrees=20, swing2LimitDegrees=1, twistLimitDegrees=1}
 {bone1Name=arm_bot_04_l, bone2Name=arm_bot_03_l, swing1LimitDegrees=20, swing2LimitDegrees=1, twistLimitDegrees=1}
 {bone1Name=arm_mid_bot_02_l, bone2Name=backpack_root, swing1LimitDegrees=20, swing2LimitDegrees=1, twistLimitDegrees=1}
 {bone1Name=arm_mid_bot_03_l, bone2Name=arm_mid_bot_02_l, swing1LimitDegrees=35, swing2LimitDegrees=1, twistLimitDegrees=1}
 {bone1Name=arm_mid_bot_04_l, bone2Name=arm_mid_bot_03_l, swing1LimitDegrees=45, swing2LimitDegrees=1, twistLimitDegrees=1}
 {bone1Name=arm_mid_top_02_l, bone2Name=backpack_root, swing1LimitDegrees=20, swing2LimitDegrees=1, twistLimitDegrees=1}
 {bone1Name=arm_mid_top_03_l, bone2Name=arm_mid_top_02_l, swing1LimitDegrees=20, swing2LimitDegrees=1, twistLimitDegrees=1}
 {bone1Name=arm_mid_top_04_l, bone2Name=arm_mid_top_03_l, swing1LimitDegrees=20, swing2LimitDegrees=1, twistLimitDegrees=1}
 {bone1Name=arm_top_04_l, bone2Name=arm_top_03_l, swing1LimitDegrees=20, swing2LimitDegrees=1, twistLimitDegrees=1}
 {bone1Name=arm_top_02_l, bone2Name=backpack_root, swing1LimitDegrees=20, swing2LimitDegrees=1, twistLimitDegrees=1}
 {bone1Name=arm_top_03_l, bone2Name=arm_top_02_l, swing1LimitDegrees=20, swing2LimitDegrees=1, twistLimitDegrees=1}
]
