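# Animation Blueprint for the Blackwidow back bling, posing its multi-jointed arms to follow the player's movement.
# A skydive-driven wingPosition state machine picks base, dive or fly poses; its Targeting state has no transitions.
# Root rotation, RigidBody physics, variable-driven bone rotations and partial Copy Bones then feed the cached preSprint pose.
# The output layers interpolated leg additives when targeting, speed-weighted additives when neither targeting nor skydiving, and preSprint alone while skydiving.
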AnimBlueprint /BRCosmetics/Accessories/FORT_Backpacks/F_MED_Blackwidow_Backpack/Meshes/F_MED_Blackwidow_Backpack_01_AnimBP
target_skeleton=/Game/Accessories/FORT_Backpacks/F_MED_Blackwidow_Backpack/Meshes/F_MED_Blackwidow_Backpack_01_Skeleton
graphs [
 AnimGraph {nodes [
   n1 = F_MED_Blackwidow_Backpack_01_Add_PoseAsset : base!(PoseWeight:float=1) -> (Pose:Pose Link Structure) {
    is=title AnimGraphNode_PoseByName
    poseAsset=&/F_MED_Blackwidow_Backpack_01_Add_PoseAsset
    poseName=base
   }
   Apply Additive(Additive=@n1) {alphaBoolBlend {blendOption=Linear}}
   n2 = Copy Bone - Source Bone: arm_mid_bot_01_r - Target Bone: arm_mid_bot_01_l!(
    ComponentPose:Component Space Pose Link Structure
    Alpha:float=0.3
    bCopyTranslation:bool=False
    bCopyRotation:bool=true
    bCopyScale:bool=False
   ) -> (Pose:Component Space Pose Link Structure) {
    is=title AnimGraphNode_CopyBone
    controlSpace=BCS_ParentBoneSpace
    alphaBoolBlend {blendOption=Linear}
    sourceBone {boneName=arm_mid_bot_01_r}
    targetBone {boneName=arm_mid_bot_01_l}
   }
   n3 = Copy Bone - Source Bone: arm_mid_bot_03_r - Target Bone: arm_mid_bot_03_l!(
    ComponentPose:Component Space Pose Link Structure=@n2
    Alpha:float=0.65
    bCopyTranslation:bool=False
    bCopyRotation:bool=true
    bCopyScale:bool=False
   ) -> (Pose:Component Space Pose Link Structure) {
    is=title AnimGraphNode_CopyBone
    controlSpace=BCS_ParentBoneSpace
    alphaBoolBlend {blendOption=Linear}
    sourceBone {boneName=arm_mid_bot_03_r}
    targetBone {boneName=arm_mid_bot_03_l}
   }
   Copy Bone - Source Bone: arm_mid_bot_04_r - Target Bone: arm_mid_bot_04_l!(
    ComponentPose:Component Space Pose Link Structure=@n3
    Alpha:float=0.2
    bCopyTranslation:bool=False
    bCopyRotation:bool=true
    bCopyScale:bool=False
   ) -> (Pose:Component Space Pose Link Structure) {
    is=title AnimGraphNode_CopyBone
    controlSpace=BCS_ParentBoneSpace
    alphaBoolBlend {blendOption=Linear}
    sourceBone {boneName=arm_mid_bot_04_r}
    targetBone {boneName=arm_mid_bot_04_l}
   }
   n4 = F_MED_Blackwidow_Backpack_01_Add_PoseAsset : legs_01!(PoseWeight:float=1) -> (Pose:Pose Link Structure) {
    is=title AnimGraphNode_PoseByName
    poseAsset=&F_MED_Blackwidow_Backpack_01_Add_PoseAsset
    poseName=legs_01
   }
   n5 = Apply Additive(Base=Use cached pose 'preSprint'(), Additive=@n4, Alpha=Get Legs01PoseTargetAlpha() -> (Legs01PoseTargetAlpha:float)) {alphaBoolBlend {blendOption=Linear}, alphaScaleBiasClamp {bInterpResult=true, interpSpeedDecreasing=3, interpSpeedIncreasing=3}}
   n6 = F_MED_Blackwidow_Backpack_01_Add_PoseAsset : legs_02!(PoseWeight:float=1) -> (Pose:Pose Link Structure) {
    is=title AnimGraphNode_PoseByName
    poseAsset=&F_MED_Blackwidow_Backpack_01_Add_PoseAsset
    poseName=legs_02
   }
   n7 = Apply Additive(Base=@n5, Additive=@n6, Alpha=Get Legs02PoseTargetAlpha() -> (Legs02PoseTargetAlpha:float)) {alphaBoolBlend {blendOption=Linear}, alphaScaleBiasClamp {bInterpResult=true, interpSpeedDecreasing=5, interpSpeedIncreasing=5}}
   n8 = F_MED_Blackwidow_Backpack_01_Add_PoseAsset : legs_03!(PoseWeight:float=1) -> (Pose:Pose Link Structure) {
    is=title AnimGraphNode_PoseByName
    poseAsset=&F_MED_Blackwidow_Backpack_01_Add_PoseAsset
    poseName=legs_03
   }
   n9 = Apply Additive(Base=@n7, Additive=@n8, Alpha=Get Legs03PoseTargetAlpha() -> (Legs03PoseTargetAlpha:float)) {alphaBoolBlend {blendOption=Linear}, alphaScaleBiasClamp {bInterpResult=true, interpSpeedDecreasing=4, interpSpeedIncreasing=4}}
   n10 = F_MED_Blackwidow_Backpack_01_Add_PoseAsset : legs_04!(PoseWeight:float=1) -> (Pose:Pose Link Structure) {
    is=title AnimGraphNode_PoseByName
    poseAsset=&F_MED_Blackwidow_Backpack_01_Add_PoseAsset
    poseName=legs_04
   }
   n11 = Apply Additive(Base=@n9, Additive=@n10, Alpha=Get Legs04PoseTargetAlpha() -> (Legs04PoseTargetAlpha:float)) {alphaBoolBlend {blendOption=Linear}, alphaScaleBiasClamp {bInterpResult=true, interpSpeedDecreasing=3, interpSpeedIncreasing=3}}
   n12 = Get PawnSpeedRange() -> (PawnSpeedRange:float)
   n13 = F_MED_Blackwidow_Backpack_01_Add_PoseAsset : back!(PoseWeight:float=@n12) -> (Pose:Pose Link Structure) {
    is=title AnimGraphNode_PoseByName
    poseAsset=&F_MED_Blackwidow_Backpack_01_Add_PoseAsset
    poseName=back
   }
   n14 = F_MED_Blackwidow_Backpack_01_Add_PoseAsset : base(PoseWeight=@n12) {poseAsset=&F_MED_Blackwidow_Backpack_01_Add_PoseAsset, poseName=base}
   n15 = Apply Additive(
    Base=Apply Additive(Base=Use cached pose 'preSprint'(), Additive=@n13, Alpha=0.6) {alphaBoolBlend {blendOption=Linear}}
    Additive=@n14
    Alpha=0.1
   ) {alphaBoolBlend {blendOption=Linear}}
   n16 = Use cached pose 'preSprint'()
   n17 = Get bIsSkyDiving()
   Output Pose(Result=Blend Poses by bool(
     BlendPose_0=@n11
     BlendPose_1=Blend Poses by bool(BlendPose_0=@n16, BlendPose_1=@n15, BlendTime_0=0.3, BlendTime_1=0.3, bActiveValue=@n17) {blendType=Linear}
     BlendTime_0=0.3
     BlendTime_1=0.3
     bActiveValue=Get bIsTargeting()
    ) {blendType=Linear}) {layerGroup=None}
   AnimGraphNode_StateMachine_0 = wingPosition() {is=title AnimGraphNode_StateMachine}
   n18 = "Transform (Modify) Bone - Bone: backpack_root"(
    ComponentPose=Local To Component(LocalPose=@AnimGraphNode_StateMachine_0)
    Alpha=1
    Translation="0,0,0"
    Rotation=Get BackpackRotator() -> (BackpackRotator:Rotator)
    Scale="1,1,1"
   ) {rotationMode=BMM_Additive, rotationSpace=BCS_BoneSpace, alphaBoolBlend {blendOption=Linear}, boneToModify {boneName=backpack_root}}
   n19 = RigidBody(ComponentPose=@n18, Alpha=0.45, ExternalForce=Get FallingAnimDynamicsUpForce() -> (FallingAnimDynamicsUpForce:Vector)) {
    bOverrideWorldGravity=true
    overridePhysicsAsset=&/F_MED_Blackwidow_Backpack_01_Physics
    alphaBoolBlend {blendOption=Linear}
    componentAppliedLinearAccClamp=<75 75 75>
    componentLinearAccScale=<1 1 1>
    componentLinearVelScale=<100 100 100>
    overrideWorldGravity {z=-600}
   }
   n20 = "Transform (Modify) Bone - Bone: arm_top_01_r"(ComponentPose=@n19, Alpha=1, Translation="0,0,0", Rotation=Get LegTop01Rotation() -> (LegTop01Rotation:Rotator), Scale="1,1,1") {rotationMode=BMM_Additive, rotationSpace=BCS_ParentBoneSpace, alphaBoolBlend {blendOption=Linear}, boneToModify {boneName=arm_top_01_r}}
   n21 = "Transform (Modify) Bone - Bone: arm_top_02_r"(ComponentPose=@n20, Alpha=1, Translation="0,0,0", Rotation=Get LegTop02Rotation() -> (LegTop02Rotation:Rotator), Scale="1,1,1") {
    rotationMode=BMM_Additive
    rotationSpace=BCS_ParentBoneSpace
    translationSpace=BCS_ParentBoneSpace
    alphaBoolBlend {blendOption=Linear}
    boneToModify {boneName=arm_top_02_r}
   }
   n22 = "Transform (Modify) Bone - Bone: arm_top_03_r"(ComponentPose=@n21, Alpha=1, Translation="0,0,0", Rotation=Get LegTop03Rotation() -> (LegTop03Rotation:Rotator), Scale="1,1,1") {
    rotationMode=BMM_Additive
    rotationSpace=BCS_ParentBoneSpace
    translationSpace=BCS_ParentBoneSpace
    alphaBoolBlend {blendOption=Linear}
    boneToModify {boneName=arm_top_03_r}
   }
   n23 = "Transform (Modify) Bone - Bone: arm_top_04_r"(ComponentPose=@n22, Alpha=1, Translation="0,0,0", Rotation=Get LegTop04Rotation() -> (LegTop04Rotation:Rotator), Scale="1,1,1") {
    rotationMode=BMM_Additive
    rotationSpace=BCS_ParentBoneSpace
    translationSpace=BCS_ParentBoneSpace
    alphaBoolBlend {blendOption=Linear}
    boneToModify {boneName=arm_top_04_r}
   }
   n24 = Copy Bone - Source Bone: arm_top_01_r - Target Bone: arm_mid_top_01_l!(
    ComponentPose:Component Space Pose Link Structure=@n23
    Alpha:float=0.6
    bCopyTranslation:bool=False
    bCopyRotation:bool=true
    bCopyScale:bool=False
   ) -> (Pose:Component Space Pose Link Structure) {
    is=title AnimGraphNode_CopyBone
    controlSpace=BCS_ParentBoneSpace
    alphaBoolBlend {blendOption=Linear}
    sourceBone {boneName=arm_top_01_r}
    targetBone {boneName=arm_mid_top_01_l}
   }
   n25 = Copy Bone - Source Bone: arm_top_03_r - Target Bone: arm_mid_top_03_l!(
    ComponentPose:Component Space Pose Link Structure=@n24
    Alpha:float=0.3
    bCopyTranslation:bool=False
    bCopyRotation:bool=true
    bCopyScale:bool=False
   ) -> (Pose:Component Space Pose Link Structure) {
    is=title AnimGraphNode_CopyBone
    controlSpace=BCS_ParentBoneSpace
    alphaBoolBlend {blendOption=Linear}
    sourceBone {boneName=arm_top_03_r}
    targetBone {boneName=arm_mid_top_03_l}
   }
   n26 = Copy Bone - Source Bone: arm_top_04_r - Target Bone: arm_mid_top_04_l!(
    ComponentPose:Component Space Pose Link Structure=@n25
    Alpha:float=0.6
    bCopyTranslation:bool=False
    bCopyRotation:bool=true
    bCopyScale:bool=False
   ) -> (Pose:Component Space Pose Link Structure) {
    is=title AnimGraphNode_CopyBone
    controlSpace=BCS_ParentBoneSpace
    alphaBoolBlend {blendOption=Linear}
    sourceBone {boneName=arm_top_04_r}
    targetBone {boneName=arm_mid_top_04_l}
   }
   n27 = Get LegMid01Rotation() -> (LegMid01Rotation:Rotator)
   n28 = "Transform (Modify) Bone - Bone: arm_top_01_l"(ComponentPose=@n26, Alpha=0.85, Translation="0,0,0", Rotation=@n27, Scale="1,1,1") {
    rotationMode=BMM_Additive
    alphaBoolBlend {blendOption=Linear}
    alphaScaleBiasClamp {bInterpResult=true, interpSpeedDecreasing=13, interpSpeedIncreasing=13}
    boneToModify {boneName=arm_top_01_l}
   }
   n29 = "Transform (Modify) Bone - Bone: arm_top_03_l"(ComponentPose=@n28, Alpha=0.85, Translation="0,0,0", Rotation=Get LegMid03Rotation() -> (LegMid03Rotation:Rotator), Scale="1,1,1") {
    rotationMode=BMM_Additive
    alphaBoolBlend {blendOption=Linear}
    alphaScaleBiasClamp {bInterpResult=true, interpSpeedDecreasing=13, interpSpeedIncreasing=13}
    boneToModify {boneName=arm_top_03_l}
   }
   n30 = "Transform (Modify) Bone - Bone: arm_top_04_l"(ComponentPose=@n29, Alpha=0.5, Translation="0,0,0", Rotation=Get LegMid04Rotation() -> (LegMid04Rotation:Rotator), Scale="1,1,1") {
    rotationMode=BMM_Additive
    alphaBoolBlend {blendOption=Linear}
    alphaScaleBiasClamp {bInterpResult=true, interpSpeedDecreasing=13, interpSpeedIncreasing=13}
    boneToModify {boneName=arm_top_04_l}
   }
   n31 = Copy Bone - Source Bone: arm_top_01_l - Target Bone: arm_mid_top_01_r!(
    ComponentPose:Component Space Pose Link Structure=@n30
    Alpha:float=0.2
    bCopyTranslation:bool=False
    bCopyRotation:bool=true
    bCopyScale:bool=False
   ) -> (Pose:Component Space Pose Link Structure) {
    is=title AnimGraphNode_CopyBone
    controlSpace=BCS_ParentBoneSpace
    alphaBoolBlend {blendOption=Linear}
    sourceBone {boneName=arm_top_01_l}
    targetBone {boneName=arm_mid_top_01_r}
   }
   n32 = Copy Bone - Source Bone: arm_top_03_l - Target Bone: arm_mid_top_03_r!(
    ComponentPose:Component Space Pose Link Structure=@n31
    Alpha:float=0.45
    bCopyTranslation:bool=False
    bCopyRotation:bool=true
    bCopyScale:bool=False
   ) -> (Pose:Component Space Pose Link Structure) {
    is=title AnimGraphNode_CopyBone
    controlSpace=BCS_ParentBoneSpace
    alphaBoolBlend {blendOption=Linear}
    sourceBone {boneName=arm_top_03_l}
    targetBone {boneName=arm_mid_top_03_r}
   }
   n33 = Copy Bone - Source Bone: arm_mid_top_04_r - Target Bone: arm_mid_top_04_r!(
    ComponentPose:Component Space Pose Link Structure=@n32
    Alpha:float=0.95
    bCopyTranslation:bool=False
    bCopyRotation:bool=true
    bCopyScale:bool=False
   ) -> (Pose:Component Space Pose Link Structure) {
    is=title AnimGraphNode_CopyBone
    controlSpace=BCS_ParentBoneSpace
    alphaBoolBlend {blendOption=Linear}
    sourceBone {boneName=arm_mid_top_04_r}
    targetBone {boneName=arm_mid_top_04_r}
   }
   n34 = Get LegBot01Rotation() -> (LegBot01Rotation:Rotator)
   n35 = "Transform (Modify) Bone - Bone: arm_mid_bot_01_r"(ComponentPose=@n33, Alpha=0.3, Translation="0,0,0", Rotation=@n34, Scale="1,1,1") {
    rotationMode=BMM_Additive
    alphaBoolBlend {blendOption=Linear}
    alphaScaleBiasClamp {bInterpResult=true, interpSpeedDecreasing=5, interpSpeedIncreasing=5}
    boneToModify {boneName=arm_mid_bot_01_r}
   }
   n36 = Get LegBot02Rotation() -> (LegBot02Rotation:Rotator)
   n37 = "Transform (Modify) Bone - Bone: arm_mid_bot_03_r"(ComponentPose=@n35, Alpha=0.7, Translation="0,0,0", Rotation=@n36, Scale="1,1,1") {
    rotationMode=BMM_Additive
    rotationSpace=BCS_ParentBoneSpace
    alphaBoolBlend {blendOption=Linear}
    alphaScaleBiasClamp {bInterpResult=true, interpSpeedDecreasing=5, interpSpeedIncreasing=5}
    boneToModify {boneName=arm_mid_bot_03_r}
   }
   n38 = "Transform (Modify) Bone - Bone: arm_mid_bot_04_r"(ComponentPose=@n37, Alpha=0.9, Translation="0,0,0", Rotation=Get LegBot03Rotation() -> (LegBot03Rotation:Rotator), Scale="1,1,1") {
    rotationMode=BMM_Additive
    alphaBoolBlend {blendOption=Linear}
    alphaScaleBiasClamp {bInterpResult=true, interpSpeedDecreasing=5, interpSpeedIncreasing=5}
    boneToModify {boneName=arm_mid_bot_04_r}
   }
   n39 = "Transform (Modify) Bone - Bone: arm_mid_bot_01_l"(ComponentPose=@n38, Alpha=0.8, Translation="0,0,0", Rotation=@n27, Scale="1,1,1") {
    rotationMode=BMM_Additive
    alphaBoolBlend {blendOption=Linear}
    alphaScaleBiasClamp {bInterpResult=true, interpSpeedDecreasing=5, interpSpeedIncreasing=5}
    boneToModify {boneName=arm_mid_bot_01_l}
   }
   n40 = "Transform (Modify) Bone - Bone: arm_mid_bot_03_l"(ComponentPose=@n39, Alpha=0.3, Translation="0,0,0", Rotation=@n36, Scale="1,1,1") {
    rotationMode=BMM_Additive
    rotationSpace=BCS_ParentBoneSpace
    alphaBoolBlend {blendOption=Linear}
    alphaScaleBiasClamp {bInterpResult=true, interpSpeedDecreasing=5, interpSpeedIncreasing=5}
    boneToModify {boneName=arm_mid_bot_03_l}
   }
   n41 = "Transform (Modify) Bone - Bone: arm_mid_bot_04_l"(ComponentPose=@n40, Alpha=0.8, Translation="0,0,0", Rotation=@n34, Scale="1,1,1") {
    rotationMode=BMM_Additive
    alphaBoolBlend {blendOption=Linear}
    alphaScaleBiasClamp {bInterpResult=true, interpSpeedDecreasing=5, interpSpeedIncreasing=5}
    boneToModify {boneName=arm_mid_bot_04_l}
   }
   Save cached pose 'preSprint'(Pose=Component To Local(ComponentPose=@n41)) {cacheName=preSprint}
  ]}
 wingPosition {
  entry=Base
  path=AnimGraph.AnimGraphNode_StateMachine_0.wingPosition
  ids=id
  nodes [
   AnimStateNode_0 = Base() {is=title AnimStateNode}
   AnimStateNode_1 = Diving() {is=title AnimStateNode}
   AnimStateNode_3 = Targeting() {is=title AnimStateNode}
   AnimStateNode_4 = Floating() {is=title AnimStateNode}
   AnimStateNode_5 = Hanging() {is=title AnimStateNode}
   AnimStateTransitionNode_0 = AnimStateTransitionNode()
   AnimStateTransitionNode_1 = AnimStateTransitionNode()
   AnimStateTransitionNode_10 = AnimStateTransitionNode()
   AnimStateTransitionNode_11 = AnimStateTransitionNode()
   AnimStateTransitionNode_12 = AnimStateTransitionNode()
   AnimStateTransitionNode_13 = AnimStateTransitionNode()
   AnimStateTransitionNode_14 = AnimStateTransitionNode()
   AnimStateTransitionNode_2 = AnimStateTransitionNode()
   AnimStateTransitionNode_3 = AnimStateTransitionNode()
   AnimStateTransitionNode_4 = AnimStateTransitionNode()
   AnimStateTransitionNode_5 = AnimStateTransitionNode()
   AnimStateTransitionNode_9 = AnimStateTransitionNode()
  ]
  links [
   {
    dst=Diving
    rule=AnimGraph.AnimGraphNode_StateMachine_0.wingPosition.AnimStateTransitionNode_0.Transition
    src=Base
    via=AnimStateTransitionNode_0
   }
   {
    dst=Floating
    rule=AnimGraph.AnimGraphNode_StateMachine_0.wingPosition.AnimStateTransitionNode_4.Transition
    src=Base
    via=AnimStateTransitionNode_4
   }
   {
    dst=Hanging
    rule=AnimGraph.AnimGraphNode_StateMachine_0.wingPosition.AnimStateTransitionNode_11.Transition
    src=Base
    via=AnimStateTransitionNode_11
   }
   {
    dst=Base
    rule=AnimGraph.AnimGraphNode_StateMachine_0.wingPosition.AnimStateTransitionNode_3.Transition
    src=Diving
    via=AnimStateTransitionNode_3
   }
   {
    dst=Floating
    rule=AnimGraph.AnimGraphNode_StateMachine_0.wingPosition.AnimStateTransitionNode_1.Transition
    src=Diving
    via=AnimStateTransitionNode_1
   }
   {
    dst=Hanging
    rule=AnimGraph.AnimGraphNode_StateMachine_0.wingPosition.AnimStateTransitionNode_10.Transition
    src=Diving
    via=AnimStateTransitionNode_10
   }
   {
    dst=Base
    rule=AnimGraph.AnimGraphNode_StateMachine_0.wingPosition.AnimStateTransitionNode_2.Transition
    src=Floating
    via=AnimStateTransitionNode_2
   }
   {
    dst=Diving
    rule=AnimGraph.AnimGraphNode_StateMachine_0.wingPosition.AnimStateTransitionNode_5.Transition
    src=Floating
    via=AnimStateTransitionNode_5
   }
   {
    dst=Hanging
    rule=AnimGraph.AnimGraphNode_StateMachine_0.wingPosition.AnimStateTransitionNode_9.Transition
    src=Floating
    via=AnimStateTransitionNode_9
   }
   {
    dst=Base
    rule=AnimGraph.AnimGraphNode_StateMachine_0.wingPosition.AnimStateTransitionNode_12.Transition
    src=Hanging
    via=AnimStateTransitionNode_12
   }
   {
    dst=Diving
    rule=AnimGraph.AnimGraphNode_StateMachine_0.wingPosition.AnimStateTransitionNode_14.Transition
    src=Hanging
    via=AnimStateTransitionNode_14
   }
   {
    dst=Floating
    rule=AnimGraph.AnimGraphNode_StateMachine_0.wingPosition.AnimStateTransitionNode_13.Transition
    src=Hanging
    via=AnimStateTransitionNode_13
   }
  ]
 }
 Base {
  path=AnimGraph.AnimGraphNode_StateMachine_0.wingPosition.AnimStateNode_0.Base
  nodes [
   n1 = F_MED_Blackwidow_Backpack_01_PoseAsset : base!(PoseWeight:float=1) -> (Pose:Pose Link Structure) {
    is=title AnimGraphNode_PoseByName
    poseAsset=&/F_MED_Blackwidow_Backpack_01_PoseAsset
    poseName=base
   }
   Output Animation Pose(Result=@n1)
  ]
 }
 Diving {
  path=AnimGraph.AnimGraphNode_StateMachine_0.wingPosition.AnimStateNode_1.Diving
  nodes [
   n1 = F_MED_Blackwidow_Backpack_01_PoseAsset : dive!(PoseWeight:float=1) -> (Pose:Pose Link Structure) {
    is=title AnimGraphNode_PoseByName
    poseAsset=&F_MED_Blackwidow_Backpack_01_PoseAsset
    poseName=dive
   }
   Output Animation Pose(Result=@n1)
  ]
 }
 Targeting {
  path=AnimGraph.AnimGraphNode_StateMachine_0.wingPosition.AnimStateNode_3.Targeting
  nodes [
   n1 = F_MED_Blackwidow_Backpack_01_PoseAsset : back!(PoseWeight:float=1) -> (Pose:Pose Link Structure) {
    is=title AnimGraphNode_PoseByName
    poseAsset=&F_MED_Blackwidow_Backpack_01_PoseAsset
    poseName=back
   }
   Output Animation Pose(Result=@n1)
  ]
 }
 Floating {
  path=AnimGraph.AnimGraphNode_StateMachine_0.wingPosition.AnimStateNode_4.Floating
  nodes [
   n1 = F_MED_Blackwidow_Backpack_01_PoseAsset : fly!(PoseWeight:float=1) -> (Pose:Pose Link Structure) {
    is=title AnimGraphNode_PoseByName
    poseAsset=&F_MED_Blackwidow_Backpack_01_PoseAsset
    poseName=fly
   }
   Output Animation Pose(Result=@n1)
  ]
 }
 Hanging {
  path=AnimGraph.AnimGraphNode_StateMachine_0.wingPosition.AnimStateNode_5.Hanging
  nodes [
   n1 = F_MED_Blackwidow_Backpack_01_PoseAsset : base(PoseWeight=1) {poseAsset=&F_MED_Blackwidow_Backpack_01_PoseAsset, poseName=base}
   Output Animation Pose(Result=@n1)
  ]
 }
 Transition {
  path=AnimGraph.AnimGraphNode_StateMachine_0.wingPosition.AnimStateTransitionNode_0.Transition
  nodes [Result(bCanEnterTransition=Get bIsSkydiveDiveMode())]
 }
 Transition {
  path=AnimGraph.AnimGraphNode_StateMachine_0.wingPosition.AnimStateTransitionNode_1.Transition
  nodes [Result(bCanEnterTransition=Get bIsSkydiveFloating())]
 }
 Transition {
  path=AnimGraph.AnimGraphNode_StateMachine_0.wingPosition.AnimStateTransitionNode_10.Transition
  nodes [Result(bCanEnterTransition=Get bIsParachuteOpened())]
 }
 Transition {
  path=AnimGraph.AnimGraphNode_StateMachine_0.wingPosition.AnimStateTransitionNode_11.Transition
  nodes [Result(bCanEnterTransition=Get bIsParachuteOpened())]
 }
 Transition {
  path=AnimGraph.AnimGraphNode_StateMachine_0.wingPosition.AnimStateTransitionNode_12.Transition
  nodes [Result(bCanEnterTransition=NOT Boolean(A=Get bIsSkyDiving()))]
 }
 Transition {
  path=AnimGraph.AnimGraphNode_StateMachine_0.wingPosition.AnimStateTransitionNode_13.Transition
  nodes [Result(bCanEnterTransition=Get bIsSkydiveFloating())]
 }
 Transition {
  path=AnimGraph.AnimGraphNode_StateMachine_0.wingPosition.AnimStateTransitionNode_14.Transition
  nodes [Result(bCanEnterTransition=Get bIsSkydiveDiveMode())]
 }
 Transition {
  path=AnimGraph.AnimGraphNode_StateMachine_0.wingPosition.AnimStateTransitionNode_2.Transition
  nodes [Result(bCanEnterTransition=NOT Boolean(A=Get bIsSkyDiving()))]
 }
 Transition {
  path=AnimGraph.AnimGraphNode_StateMachine_0.wingPosition.AnimStateTransitionNode_3.Transition
  nodes [
   Result(bCanEnterTransition=NOT Boolean(A=Get bIsSkydiveDiveMode()))
   OR Boolean(A=false, B=Get bIsParachuteOpened())
  ]
 }
 Transition {
  path=AnimGraph.AnimGraphNode_StateMachine_0.wingPosition.AnimStateTransitionNode_4.Transition
  nodes [Result(bCanEnterTransition=Get bIsSkydiveFloating())]
 }
 Transition {
  path=AnimGraph.AnimGraphNode_StateMachine_0.wingPosition.AnimStateTransitionNode_5.Transition
  nodes [Result(bCanEnterTransition=Get bIsSkydiveDiveMode())]
 }
 Transition {
  path=AnimGraph.AnimGraphNode_StateMachine_0.wingPosition.AnimStateTransitionNode_9.Transition
  nodes [Result(bCanEnterTransition=Get bIsParachuteOpened())]
 }
 EventGraph {links [], nodes []}
]
variables []
